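# Animation Blueprint for the HonorBraceLeap backpack, repositioning its pack bone as the character's body pose and movement state change.
# Each update derives spine and per-clavicle alphas from bone rotations relative to the reference pose, plus a crouch-moving flag.
# A chain of additive pack-bone offsets is cached as StartPose, then blended for skydiving and moving while crouched.
# Further additive offsets finish the pose, including one eased in while skydive floating.
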AnimBlueprint /BRCosmetics/Accessories/FORT_Backpacks/BackPack_M_MED_HonorBraceLeap/Meshes/M_MED_HonorBraceLeap_Pack_AnimBP
target_skeleton=&/M_MED_HonorBraceLeap_Pack_Skeleton
graphs [
 AnimGraph {nodes [
   n1 = "Transform (Modify) Bone - Bone: pack"(ComponentPose=Local To Component(LocalPose=Use cached pose 'StartPose'()), Translation="0,-2.5,1", Rotation="0,0,0") {
    rotationMode=BMM_Additive
    scaleMode=BMM_Replace
    translationMode=BMM_Additive
    alphaBoolBlend {blendOption=Linear}
    boneToModify {boneName=pack}
    translation {y=-2.5, z=1}
   }
   n2 = Blend Poses by bool(
    BlendPose_0=Component To Local(ComponentPose=@n1)
    BlendPose_1=Use cached pose 'StartPose'()
    BlendTime_0=0.4
    BlendTime_1=0.3
    bActiveValue=Get bIsSkyDiving()
   ) {blendType=Linear}
   n3 = Use cached pose 'StartPose'()
   n4 = Component To Local(ComponentPose="Transform (Modify) Bone - Bone: pack"!(ComponentPose=Local To Component(LocalPose=@n3), Alpha, Rotation="0,0,-6") -> (Pose) {rotationMode=BMM_Additive, alphaBoolBlend {blendOption=Linear}, boneToModify {boneName=pack}, rotation {roll=-6}})
   n5 = Get CrouchMoving() -> (CrouchMoving:bool)
   n6 = "Transform (Modify) Bone - Bone: pack"!(
    ComponentPose=Local To Component(LocalPose=Blend Poses by bool(BlendPose_0=@n4, BlendPose_1=@n2, BlendTime_0=0.5, BlendTime_1=0.5, bActiveValue=@n5) {blendType=Linear})
    Alpha
    Translation:Vector="0,-2,2"
    Rotation="0,0,30"
    Scale:Vector
   ) -> (Pose) {
    rotationMode=BMM_Additive
    scaleMode=BMM_Additive
    scaleSpace=BCS_BoneSpace
    translationMode=BMM_Additive
    alphaBoolBlend {blendOption=Linear}
    boneToModify {boneName=pack}
    rotation {roll=30}
    translation {y=-2, z=2}
   }
   n7 = "Transform (Modify) Bone - Bone: pack"!(ComponentPose=@n6, bAlphaBoolEnabled:bool=Get bIsSkydiveFloating(), Translation="0,1,0", Rotation="0,0,-3", Scale) -> (Pose) {
    alphaInputType=Bool
    rotationMode=BMM_Additive
    scaleMode=BMM_Additive
    scaleSpace=BCS_BoneSpace
    translationMode=BMM_Additive
    alphaBoolBlend {blendInTime=0.2, blendOption=Linear, blendOutTime=0.8}
    boneToModify {boneName=pack}
    rotation {roll=-3}
    translation {y=1}
   }
   Output Pose(Result=Component To Local(ComponentPose=@n7)) {layerGroup=None}
   n8 = "Transform (Modify) Bone - Bone: pack"!(ComponentPose=Local To Component(LocalPose=Input Pose()), Alpha:float, Translation="0,1,0", Rotation="0,0,0", Scale) -> (Pose) {
    rotationMode=BMM_Additive
    scaleMode=BMM_Replace
    translationMode=BMM_Additive
    alphaBoolBlend {blendOption=Linear}
    boneToModify {boneName=pack}
    translation {y=1}
   }
   n9 = "Transform (Modify) Bone - Bone: pack"(ComponentPose=@n8, Translation="2,-3,0", Rotation="0,10,0") {
    rotationMode=BMM_Additive
    rotationSpace=BCS_BoneSpace
    translationMode=BMM_Additive
    translationSpace=BCS_BoneSpace
    alphaBoolBlend {blendInTime=0.3, blendOption=ExpInOut, blendOutTime=0.7}
    alphaScaleBiasClamp {interpSpeedDecreasing=20, interpSpeedIncreasing=20}
    boneToModify {boneName=pack}
    rotation {yaw=10}
    translation=<2 -3>
   }
   n10 = "Transform (Modify) Bone - Bone: pack"(ComponentPose=@n9, Translation="-3,-3,0", Rotation="0,-10,0") {
    rotationMode=BMM_Additive
    rotationSpace=BCS_BoneSpace
    translationMode=BMM_Additive
    translationSpace=BCS_BoneSpace
    alphaBoolBlend {blendInTime=0.3, blendOption=ExpInOut, blendOutTime=0.7}
    alphaScaleBiasClamp {interpSpeedDecreasing=20, interpSpeedIncreasing=20}
    boneToModify {boneName=pack}
    rotation {yaw=-10}
    translation=<-3 -3>
   }
   Save cached pose 'StartPose'(Pose=Component To Local(ComponentPose=@n10)) {cacheName=StartPose}
  ]}
 EventGraph {nodes [
   Event BlueprintUpdateAnimation()
   > n1 = Sequence!(execute) -> (then_0, then_1, then_2:Exec)
   n2 = GetOwnerSkeletalMeshComponent()
   n3 = GetDeltaTransformFromRefPose(self=@n2, BoneName=clavicle_r, BaseName=spine_05)
   Set RightArmBackward_Alpha(
    execute=@n1.then_1
    RightArmBackward_Alpha:double=MapRangeClamped(
     Value=BreakTransform!(InTransform=@n3) -> (Location, Rotation_Roll:float, Rotation_Pitch:float, Rotation_Yaw:float, Scale).Rotation_Yaw
     InRangeA=0
     InRangeB=-30
    )
   ) -> (Output_Get:double)
   n4 = GetDeltaTransformFromRefPose(self=@n2, BoneName=clavicle_l, BaseName=spine_05)
   > Set LeftArmBackward_Alpha(LeftArmBackward_Alpha:double=MapRangeClamped(Value=BreakTransform(InTransform=@n4).Rotation_Yaw, InRangeA=0, InRangeB=-30)) -> (Output_Get:double)
   Set CrouchMoving(execute=@n1.then_2, CrouchMoving:bool=AND Boolean(A=Get bIsCrouching(), B=Get bIsMoving2D())) -> (Output_Get:bool)
   n5 = BreakTransform!(InTransform=GetDeltaTransformFromRefPose(self=GetOwnerSkeletalMeshComponent(), BoneName=spine_05, BaseName=pelvis)) -> (Location, Rotation:Rotator, Scale)
   Set backpack_rotation_Alpha(execute=@n1.then_0, backpack_rotation_Alpha:double=MapRangeClamped(Value=BreakRotator(InRot=@n5.Rotation).Yaw, InRangeA=0, InRangeB=-80)) -> (Output_Get:double)
  ]}
]
variables [
 backpack_rotation_Alpha {type=number}
 CrouchMoving {type=boolean}
 LeftArmBackward_Alpha {type=number}
 RightArmBackward_Alpha {type=number}
]
